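# Unlit translucent material for the cone effect on the Skyfire backpack cosmetic, built for Niagara mesh particles and previewed on the backpack's hologram mesh.
# Particle color RGB runs through a shared emissive-adjustment function to form the glow.
# Particle alpha is multiplied by the cubed red vertex color channel, so mesh-painted values give a sharpened opacity falloff without textures.
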Material /Skyfire/Cosmetics/Backbling/Effects/M_BAckPackConeFX
graph {
 properties {
  bUsedWithNiagaraMeshParticles=true
  blendMode=BLEND_Translucent
  pixelDepthOffsetMode=PDOM_Legacy
  previewMesh=&/S_SkyFire_BackPack_hologram
  shadingModel=MSM_Unlit
  usedShadingModels=Unlit
 }
 nodes [
  n1 = ParticleColor()
  MaterialFunctionCall(Emissive=@n1.RGB) {materialFunction=&MF_FX_EmissiveAdjustments}
  Multiply(A=@n1.A, B=Power(Base=VertexColor().R) {constExponent=3})
 ]
}
