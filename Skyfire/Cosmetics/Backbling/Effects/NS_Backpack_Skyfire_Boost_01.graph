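# Niagara boost exhaust effect for the Skyfire backbling, drawing blue flames, jets and glows at two skeletal-mesh sockets picked by spawn group.
# Three emitters are active: short-lived Flame meshes, velocity-driven Jet sprites aligned to socket rotation, and two socket-following Lights with sprite, light and component renderers.
# User parameters drive flame scale, jet spawn rate and size, and light intensity.
# Older exhaust, looping-jet and light emitters remain disabled.
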NiagaraSystem /Skyfire/Cosmetics/Backbling/Effects/NS_Backpack_Skyfire_Boost_01
dependencies {
 usedDataInterfaces [
  &NiagaraDataInterfaceCurve
  &NiagaraDataInterfaceVectorCurve
  &NiagaraDataInterfaceColorCurve
  &NiagaraDataInterfaceVector2DCurve
 ]
 usedDynamicInputs []
 usedModules [&SystemState]
}
emitters {
 ExhaustFlame_Small {
  bEnabled=false
  emitterUpdateScript [
   &EmitterState()
   &SpawnRate(
    SpawnRate=&FloatFromCurve(FloatCurve=DataInterface() {propertyValues {Curve {keys [<at 0 1>,<at 0.5 0>]}}}, CurveIndex=$Emitter.NormalizedLoopAge, Scale Curve=10)
    Use Spawn Probability=false
   )
   /SetVariables(
    Emitter.Color_A:LinearColor=&MultiplyLinearColor(
     Linear Color=&FN_Select_Color_from_Time_Of_Day(
      Morning=$User.Color_01
      Day=$User.Color_01
      Evening=$User.Color_01
      Night=&MultiplyLinearColor(Linear Color=$User.Color_01, Scale Factor=<0 0 0 0.5>)
     )
     Scale Factor=&MakeFloat4FromVectorAndFloat(
      "Vector (RGB)"=&FN_FrontEndVector3Modifier(
       Gameplay=&Lerp_Vector(A=<0.5 0.5 0.5>, B=<0.5 0.5 0.5>, Alpha=$System.FN_Integrator.NormalizedValue)
       Front End=<1 1 1>
      )
      "Float (Alpha)"=0
     )
    )
    Emitter.Color_B:LinearColor=&MultiplyLinearColor(
     Linear Color=&FN_Select_Color_from_Time_Of_Day(
      Morning=$User.Color_02
      Day=$User.Color_02
      Evening=$User.Color_02
      Night=&MultiplyLinearColor(Linear Color=$User.Color_02, Scale Factor=<1 1 0.5 0.5>)
     )
     Scale Factor=&MakeFloat4FromVectorAndFloat(
      "Vector (RGB)"=&FN_FrontEndVector3Modifier(
       Gameplay=&Lerp_Vector(A=<0.5 0.5 0.5>, B=<0.5 0.5 0.5>, Alpha=$System.FN_Integrator.NormalizedValue)
       Front End=<1 1 1>
      )
      "Float (Alpha)"=0
     )
    )
   ) {isSetParameters=true, module=SetVariables_#1}
  ]
  particleSpawnScript [
   /Niagara/Modules/Spawn/Initialization/InitializeParticle(
    Color=<rgba 2 2 2 0.8>
    Mesh Scale=<0.03 0.03 0.055>
    Module.Write Color=true
    Module.Write Lifetime=true
    Module.Write Mass=false
    Module.Write Material Random=false
    Module.Write Position=false
    Module.Write Scale=true
    Module.Write SpriteRotation=false
    Module.Write SpriteSize=false
    Position=&SimulationPosition()
    Sprite Size=<5 5>
   )
   &InitialMeshOrientation(Orientation Axis=<0 1 0>, Orientation Vector=<0 0 -1>, Use Orientation Vector=true, Use Rotation Vector=false) {module=InitialMeshOrientation001}
   &DynamicMaterialParameters(
    Index 0 Param 1=&RandomRangeFloat(Minimum=0, Maximum=1)
    Index 0 Param 4=1
    Module.Param0WriteEnabled=true
    Module.Param1WriteEnabled=false
    Module.Param2WriteEnabled=false
    Module.Param3WriteEnabled=true
   )
   &FN_SetSpawnGroup(Number of Spawn Groups=2)
  ]
  particleUpdateScript [
   &ParticleState()
   &SkeletalMeshLocation(
    MaskSpawnGroup=false
    Mesh Sampling Type=::NewEnumerator5
    Sampled Position Offset=<-1.5 2 0>
    Skeletal Mesh=$User.SkeletalMesh
    Socket Sampling Mode=::NewEnumerator5
    Use Sampled Mesh Attribute as Ribbon ID=false
   )
   &ScaleColor(Scale RGB=&VectorFromFloat(Value=25), ScaleA=false, ScaleRGB=true)
   &ScaleMeshSize(Scale Factor=<1.15 1.15 2.2>)
   /SetVariables(Particles.InnerFlameScale:Vector3f=&Multiply_Vector(A=$Particles.Scale, B=<0.32 0.2 0.2>)) {isSetParameters=true, module=SetVariables_#2}
   &FN_Mesh_Attributes(
    Create Scale Attribute=true
    Scale=&Multiply_Vector(A=$Particles.Scale, B=<1 1 0.5>)
    Write to Module Namespace=true
   )
   &FN_Sprite_Attributes(Camera Offset=20, Create Camera Offset Attribute=true, Write to Module Namespace=true)
   &ScaleMeshSize(Scale Factor=&VectorFromFloat(Value=&FloatFromCurve(FloatCurve=DataInterface() {propertyValues {Curve {keys [
          {arriveTangent=9.0725, interpMode=RCIM_Cubic, leaveTangent=9.0725, tangentMode=RCTM_User, time=0, value=0}
          {arriveTangent=-2.5249, interpMode=RCIM_Cubic, leaveTangent=-2.5249, tangentMode=RCTM_User, time=0.1408, value=0.9847}
          {arriveTangent=-0.0405, interpMode=RCIM_Cubic, leaveTangent=-0.0405, tangentMode=RCTM_User, time=1, value=0}
         ]}}}))) {module=ScaleMeshSize001}
  ]
  rendererClasses [&NiagaraMeshRendererProperties]
  renderers [{rendererClass=&NiagaraMeshRendererProperties, rendererIndex=0}]
 }
 ExhaustFlame_Small001 {
  bEnabled=false
  emitterUpdateScript [
   &EmitterState()
   &SpawnRate(
    SpawnRate=&FloatFromCurve(FloatCurve=DataInterface() {propertyValues {Curve {keys [<at 0 1>,<at 0.5 0>]}}}, CurveIndex=$Emitter.NormalizedLoopAge, Scale Curve=10)
    Use Spawn Probability=false
   )
   /SetVariables(
    Emitter.Color_A:LinearColor=&MultiplyLinearColor(
     Linear Color=&FN_Select_Color_from_Time_Of_Day(
      Morning=$User.Color_01
      Day=$User.Color_01
      Evening=$User.Color_01
      Night=&MultiplyLinearColor(Linear Color=$User.Color_01, Scale Factor=<1 1 0.5 0.5>)
     )
     Scale Factor=&MakeFloat4FromVectorAndFloat(
      "Vector (RGB)"=&FN_FrontEndVector3Modifier(Gameplay=&Lerp_Vector(A=<0.5 0.5 0.5>, B=<1 1 1>, Alpha=$System.FN_Integrator.NormalizedValue), Front End=<1 1 1>)
      "Float (Alpha)"=1
     )
    )
    Emitter.Color_B:LinearColor=&MultiplyLinearColor(
     Linear Color=&FN_Select_Color_from_Time_Of_Day(
      Morning=$User.Color_02
      Day=$User.Color_02
      Evening=$User.Color_02
      Night=&MultiplyLinearColor(Linear Color=$User.Color_02, Scale Factor=<1 1 0.5 0.5>)
     )
     Scale Factor=&MakeFloat4FromVectorAndFloat(
      "Vector (RGB)"=&FN_FrontEndVector3Modifier(Gameplay=&Lerp_Vector(A=<0.5 0.5 0.5>, B=<1 1 1>, Alpha=$System.FN_Integrator.NormalizedValue), Front End=<1 1 1>)
      "Float (Alpha)"=1
     )
    )
   ) {isSetParameters=true, module=SetVariables_#1}
  ]
  particleSpawnScript [
   &InitializeParticle(
    Color=<rgba 2 2 2 0.8>
    Mesh Scale=<0.03 0.03 0.055>
    Module.Write Color=true
    Module.Write Lifetime=true
    Module.Write Mass=false
    Module.Write Material Random=false
    Module.Write Position=false
    Module.Write Scale=true
    Module.Write SpriteRotation=false
    Module.Write SpriteSize=false
    Position=&SimulationPosition()
    Sprite Size=<5 5>
   )
   &InitialMeshOrientation(Orientation Axis=<0 1 0>, Orientation Vector=<0 0 -1>, Use Orientation Vector=true, Use Rotation Vector=false) {module=InitialMeshOrientation001}
   &DynamicMaterialParameters(
    Index 0 Param 1=&RandomRangeFloat(Minimum=0, Maximum=1)
    Index 0 Param 4=1
    Module.Param0WriteEnabled=true
    Module.Param1WriteEnabled=false
    Module.Param2WriteEnabled=false
    Module.Param3WriteEnabled=true
   )
   &FN_SetSpawnGroup(Number of Spawn Groups=2)
  ]
  particleUpdateScript [
   &ParticleState()
   &SkeletalMeshLocation(
    Bone / Socket Index=1
    MaskSpawnGroup=false
    Mesh Sampling Type=::NewEnumerator5
    Sampled Position Offset=<1.5 2 1>
    Skeletal Mesh=$User.SkeletalMesh
    Socket Sampling Mode=::NewEnumerator5
    Use Sampled Mesh Attribute as Ribbon ID=false
   )
   &ScaleColor(Scale RGB=&VectorFromFloat(Value=25), ScaleA=false, ScaleRGB=true)
   &ScaleMeshSize(Scale Factor=<1.15 1.15 2.2>)
   /SetVariables(Particles.InnerFlameScale:Vector3f=&Multiply_Vector(A=$Particles.Scale, B=<0.32 0.2 0.2>)) {isSetParameters=true, module=SetVariables_#2}
   &FN_Mesh_Attributes(
    Create Scale Attribute=true
    Scale=&Multiply_Vector(A=$Particles.Scale, B=<1 1 0.5>)
    Write to Module Namespace=true
   )
   &FN_Sprite_Attributes(Camera Offset=20, Create Camera Offset Attribute=true, Write to Module Namespace=true)
   &ScaleMeshSize(Scale Factor=&VectorFromFloat(Value=&FloatFromCurve(FloatCurve=DataInterface() {propertyValues {Curve {keys [
          {arriveTangent=9.0725, interpMode=RCIM_Cubic, leaveTangent=9.0725, tangentMode=RCTM_User, time=0, value=0}
          {arriveTangent=-2.5249, interpMode=RCIM_Cubic, leaveTangent=-2.5249, tangentMode=RCTM_User, time=0.1408, value=0.9847}
          {arriveTangent=-0.0405, interpMode=RCIM_Cubic, leaveTangent=-0.0405, tangentMode=RCTM_User, time=1, value=0}
         ]}}}))) {module=ScaleMeshSize001}
  ]
  rendererClasses [&NiagaraMeshRendererProperties]
  renderers [{rendererClass=&NiagaraMeshRendererProperties, rendererIndex=0}]
 }
 Flame {
  emitterUpdateScript [
   &EmitterState(Enable Distance Culling=true, MaxDistance=8000, Module.bUseMaxDistance=true, Module.bUseMinDistance=false, Scalability Mode=::NewEnumerator1)
   &SpawnBurst_Instantaneous(Spawn Count=2, Use Spawn Probability=false) {enabled=false}
   &SpawnRate(SpawnRate=30, Use Spawn Probability=false)
  ]
  particleSpawnScript [
   /SetVariables(Particles.MeshOrientation=&MakeQuatFromAxisAngle(
     Axis=&RandomRangeVector(Minimum=<0 0 0>, Maximum=<1 0 0>, Evaluation Type=::NewEnumerator1)
     AngleInDegrees=90
    )) {enabled=false, isSetParameters=true}
   &InitializeParticle(
    Color=<rgba 0.2951 0.7601 3 0.05>
    Lifetime=0.15
    Mesh Scale=&MakeVector(
     X=&FloatFromCurve(
      FloatCurve=DataInterface() {propertyValues {Curve {keys [
          {tangentMode=RCTM_User, time=0, value=1}
          {arriveTangent=-0.0405, interpMode=RCIM_Cubic, leaveTangent=-0.0405, tangentMode=RCTM_User, time=0.7009, value=1}
          {arriveTangent=-8.0089, interpMode=RCIM_Cubic, leaveTangent=-8.0089, tangentMode=RCTM_User, time=1, value=0}
         ]}}}
      CurveIndex=$Emitter.NormalizedLoopAge
      Scale Curve=0.16
     )
     Y=0.55
     Z=0.55
    )
    Module.Write Color=true
    Module.Write Lifetime=true
    Module.Write Mass=true
    Module.Write Material Random=true
    Module.Write Position=false
    Module.Write Scale=true
    Module.Write SpriteRotation=false
    Module.Write SpriteSize=false
   )
   &FN_SetSpawnGroup(Number of Spawn Groups=2)
   &InitialMeshOrientation(Orientation Axis=<0 1 0>, Orientation Vector=<-1 0 0>, Use Orientation Vector=true, Use Rotation Vector=false) {module=InitialMeshOrientation001}
   &CameraOffset(Camera Offset Amount=-5)
  ]
  particleUpdateScript [
   &ParticleState()
   &SkeletalMeshLocation(
    Apply Sampled Orientation to Mesh Particles=true
    Bone / Socket Index=&SpawnGroup()
    Inherit Velocity=false
    Interpolate Mesh Sampling=false
    MaskSpawnGroup=false
    Mesh Sampling Type=::NewEnumerator5
    Skeletal Mesh=$User.SkeletalMesh
    Socket Sampling Mode=::NewEnumerator5
   )
   &ScaleColor(
    Scale Alpha=&FloatFromCurve(FloatCurve=DataInterface() {propertyValues {Curve {keys [<at 0 0 RCIM_Cubic>,<at 0.5 1 RCIM_Cubic>,<at 1 0 RCIM_Cubic>]}}}, Scale Curve=2)
    ScaleA=true
    ScaleRGB=true
   )
   &DynamicMaterialParameters(
    Index 0 Param 1=1
    Index 0 Param 2=5
    Index 0 Param 3=20
    Index 0 Param 4=1
    Module.Param0WriteEnabled=true
    Module.Param1WriteEnabled=true
    Module.Param2WriteEnabled=true
    Module.Param3WriteEnabled=true
   )
   &ScaleMeshSize(
    Initial Mesh Scale:Vector3f=$Particles.Scale
    Scale Factor=&self:RemapVector(
     Value:Vector3f=$User.FlameScale
     MinInput:Vector3f=<0 0 0>
     MaxInput:Vector3f=<1 1 1>
     MinOutput:Vector3f=<0.5 0 0>
     MaxOutput:Vector3f=<2 1 1>
    )
   ) {module=ScaleMeshSize001}
   &ScaleMeshSize(Scale Factor=&VectorFromCurve(VectorCurve=DataInterface() {propertyValues {
       XCurve {keys [{interpMode=RCIM_Cubic, tangentMode=RCTM_Break, time=0, value=0.75}, {interpMode=RCIM_Cubic, tangentMode=RCTM_Break, time=1, value=2}]}
       YCurve {keys [{interpMode=RCIM_Cubic, tangentMode=RCTM_Break, time=0, value=1}, {interpMode=RCIM_Cubic, tangentMode=RCTM_Break, time=1, value=0.9}]}
       ZCurve {keys [{interpMode=RCIM_Cubic, tangentMode=RCTM_Break, time=0, value=1}, {interpMode=RCIM_Cubic, tangentMode=RCTM_Break, time=1, value=0.9}]}
      }}))
  ]
  rendererClasses [&NiagaraMeshRendererProperties]
  renderers [{rendererClass=&NiagaraMeshRendererProperties, rendererIndex=0}]
 }
 Jet {
  emitterUpdateScript [
   &EmitterState(Enable Distance Culling=true, MaxDistance=2000, Module.bUseMaxDistance=true, Module.bUseMinDistance=false, Scalability Mode=::NewEnumerator1)
   &SpawnRate(
    SpawnRate=&self:RemapFloat(Value:NiagaraFloat=$User.SpawnRate, InputMin:NiagaraFloat=0, InputMax:NiagaraFloat=1, OutputMin:NiagaraFloat=20, OutputMax:NiagaraFloat=90)
    Use Spawn Probability=false
   )
  ]
  particleSpawnScript [
   &InitializeParticle(
    Color=<rgba 0.5069 2.4082 5 0.2>
    Lifetime=0.25
    Module.Write Color=true
    Module.Write Lifetime=true
    Module.Write Mass=true
    Module.Write Material Random=true
    Module.Write Position=true
    Module.Write Scale=false
    Module.Write SpriteRotation=false
    Module.Write SpriteSize=true
    Sprite Size=<8.5 25>
   )
   &FN_SetSpawnGroup(Number of Spawn Groups=2)
   &SkeletalMeshLocation(
    Apply Sampled Orientation to Mesh Particles=true
    Bone / Socket Index=&SpawnGroup()
    Inherit Velocity=false
    Interpolate Mesh Sampling=false
    MaskSpawnGroup=false
    Mesh Sampling Type=::NewEnumerator5
    Sampled Position Offset Coordinate Space:ENiagaraMeshTransforms=::NewEnumerator2
    Skeletal Mesh=$User.SkeletalMesh
    Socket Sampling Mode=::NewEnumerator5
    Write Sampled Mesh Attrributes to Particle Attributes=true
   ) {module=SkeletalMeshLocation001}
   &AddVelocity(Velocity=&RotateVectorByQuaternion(
     Vector To Rotate=&MakeVector(
      X=0
      Y=&FloatFromCurve(
       FloatCurve=DataInterface() {propertyValues {Curve {keys [
           {tangentMode=RCTM_User, time=0, value=1}
           {arriveTangent=-0.9089, interpMode=RCIM_Cubic, leaveTangent=-0.9089, tangentMode=RCTM_User, time=0.5, value=1}
           {arriveTangent=-5.8198, interpMode=RCIM_Cubic, leaveTangent=-5.8198, tangentMode=RCTM_User, time=1, value=0}
          ]}}}
       CurveIndex=$Emitter.NormalizedLoopAge
       Scale Curve=200
      )
      Z=0
     )
     Quaternion=$Output.SkeletalMeshLocation001.SampledRotation
    ))
   &CameraOffset(Camera Offset Amount=-5)
  ]
  particleUpdateScript [
   &SolveForcesAndVelocity(Clamp Velocity=false, Limit Acceleration=false)
   &ParticleState()
   &SkeletalMeshLocation(
    Apply Sampled Orientation to Mesh Particles=true
    Bone / Socket Index=&SpawnGroup()
    Inherit Velocity=false
    Interpolate Mesh Sampling=false
    MaskSpawnGroup=false
    Mesh Sampling Type=::NewEnumerator5
    Sampled Position Offset=&Lerp_Vector(
     A=&Lerp_Vector(A=<0 0 10>, B=<0 0 10>, Alpha=$User.SpriteFlame)
     B=&Lerp_Vector(
      A=&MakeVector(X=0, Y=0, Z=0)
      B=&MakeVector(X=0, Y=0, Z=&RandomRangeFloat(Minimum=-7, Maximum=-12))
      Alpha=$User.SpriteFlame
     )
     Alpha=$Particles.NormalizedAge
    )
    Sampled Position Offset Coordinate Space:ENiagaraMeshTransforms=::NewEnumerator2
    Skeletal Mesh=$User.SkeletalMesh
    Socket Sampling Mode=::NewEnumerator5
    Write Sampled Mesh Attrributes to Particle Attributes=true
   ) {enabled=false}
   /SetVariables(Particles.SpriteAlignment=&RotateVectorByQuaternion(Vector To Rotate=<0 1 0>, Quaternion=$Output.SkeletalMeshLocation.SampledRotation)) {enabled=false, isSetParameters=true}
   &ScaleSpriteSize(Scale Factor=&Vector2DFromCurve(
     Vector2Curve=DataInterface() {propertyValues {XCurve {keys [<at 0 1 RCIM_Cubic>,<at 1 0.25 RCIM_Cubic>]}, YCurve {keys [<at 0 1 RCIM_Cubic>,<at 1 1.25 RCIM_Cubic>]}}}
     ScaleCurve=&Vector2DFromFloat(Value=&Lerp_Float(A=0, B=1, Alpha=$User.SpriteFlame))
    ))
   &ScaleColor(
    Scale Alpha=&FloatFromCurve(FloatCurve=DataInterface() {propertyValues {Curve {keys [{arriveTangent=5.8059, interpMode=RCIM_Cubic, time=0, value=0}, <at 0.25 1 RCIM_Cubic>, <at 1 0 RCIM_Cubic>]}}})
    Scale RGB=&VectorFromFloat(Value=&FloatFromCurve(
      FloatCurve=DataInterface() {propertyValues {Curve {keys [{interpMode=RCIM_Cubic, tangentMode=RCTM_User, time=0, value=1}, {interpMode=RCIM_Cubic, tangentMode=RCTM_User, time=1, value=0}]}}}
      Scale Curve=6
     ))
    ScaleA=true
    ScaleRGB=true
   )
  ]
  rendererClasses [&NiagaraSpriteRendererProperties]
  renderers [{rendererClass=&NiagaraSpriteRendererProperties}]
 }
 Jet_Looping {
  bEnabled=false
  emitterUpdateScript [
   &EmitterState()
   &SpawnRate(SpawnRate=40, Use Spawn Probability=false)
   &SpawnRate(SpawnGroup=1, SpawnRate=40, Use Spawn Probability=false) {module=SpawnRate001}
  ]
  particleSpawnScript [
   /Niagara/Modules/Spawn/Initialization/V2/InitializeParticle(
    Lifetime Max=0.25
    Lifetime Min=0.15
    Lifetime Mode=::NewEnumerator1
    Mesh Scale Max=<0.2 0.25 0.25>
    Mesh Scale Min=<0.15 0.2 0.2>
    Mesh Scale Mode=::NewEnumerator2
    UsePositionOffset=false
    Write Material Random=true
   ) {module=InitializeParticle001}
   &InitialMeshOrientation(
    Orientation Coordinate Space=::Local
    Orientation Vector=&MaskVectorBySpawnGroup(Masked Vector=<-0.5 -0.25 -1>, Passthrough Vector=<0.5 -0.25 -1>)
    Rotation=&RandomRangeVector(Minimum=<0 0 0>, Maximum=<1 0 0>)
    Use Orientation Vector=true
    Use Rotation Vector=true
   )
   &SystemLocation(MaskSpawnGroup=false, Offset=&MaskVectorBySpawnGroup(Masked Vector=<-7 -5 -17>, Passthrough Vector=<7 -5 -17>))
   &CameraOffset(Camera Offset Amount=-10)
  ]
  particleUpdateScript [
   &ParticleState()
   &Color(
    Color=<rgba 0.15 0.3875 3 1>
    Scale Color=&VectorFromFloat(Value=&FloatFromCurve(
      FloatCurve=DataInterface() {propertyValues {Curve {keys [<at 0 1>,<at 1 0>]}}}
      Scale Curve=&FloatFromCurve(
       FloatCurve=DataInterface() {propertyValues {Curve {keys [<at 0 0>,<at 0.05 8>,<at 0.2 1>,<at 0.4 0.25>,<at 1 0>]}}}
       CurveIndex=$Emitter.NormalizedLoopAge
      )
     ))
   )
   &ScaleMeshSize(Scale Factor=&VectorFromCurve(VectorCurve=DataInterface() {propertyValues {
       XCurve {keys [
         {arriveTangent=5.0383, interpMode=RCIM_Cubic, leaveTangent=5.0383, tangentMode=RCTM_User, time=0, value=0.47}
         {arriveTangent=1.03, interpMode=RCIM_Cubic, leaveTangent=1.03, time=0.7, value=2.5}
         <at 1 1.5 RCIM_Cubic>
        ]}
       YCurve {keys [{arriveTangent=6.1645, interpMode=RCIM_Cubic, leaveTangent=6.1645, tangentMode=RCTM_User, time=0.0035, value=0.4903}, <at 0.9786 1.29>]}
       ZCurve {keys [
         <at -0.0011 1.0067 RCIM_Cubic>
         {arriveTangent=-1.635, interpMode=RCIM_Cubic, leaveTangent=-1.635, tangentMode=RCTM_User, time=1.0016, value=0.28}
        ]}
      }}))
  ]
  rendererClasses [&NiagaraMeshRendererProperties]
  renderers [{rendererClass=&NiagaraMeshRendererProperties, rendererIndex=0}]
 }
 Light {
  bEnabled=false
  emitterUpdateScript [
   &EmitterState()
   &SpawnBurst_Instantaneous(Spawn Count=2, Use Spawn Probability=false)
  ]
  particleSpawnScript [
   /Niagara/Modules/Spawn/Initialization/InitializeParticle(
    Color=<rgba 5 6.5 20 1>
    Lifetime=1.25
    Module.Write Color=true
    Module.Write Lifetime=true
    Module.Write Mass=true
    Module.Write Material Random=true
    Module.Write Position=false
    Module.Write Scale=false
    Module.Write SpriteRotation=false
    Module.Write SpriteSize=true
    Position=&SimulationPosition()
    Sprite Size=<20 20>
   )
   &FN_SetSpawnGroup(Number of Spawn Groups=2)
   &SystemLocation(MaskSpawnGroup=false, Offset=&MaskVectorBySpawnGroup(Masked Vector=<-8 0 -22>, Passthrough Vector=<8 0 -22>)) {module=SystemLocation001}
  ]
  particleUpdateScript [
   &ParticleState()
   &SolveForcesAndVelocity(Clamp Velocity=false, Limit Acceleration=false)
   &ScaleColor(
    Scale RGB=&VectorFromFloat(Value=&FloatFromCurve(
      FloatCurve=DataInterface() {propertyValues {Curve {keys [<at 0 0.5>,<at 0.1 0.85>,<at 0.2 0.6>,<at 0.4 0.85>,<at 0.5 0.5>,<at 0.6 0.9>,<at 0.8 0.6>,<at 0.9 0.85>,<at 1 0.5>]}}}
      Scale Curve=&FloatFromCurve(FloatCurve=DataInterface() {propertyValues {Curve {keys [<at 0 1>,<at 1 1>]}}})
     ))
    ScaleA=true
    ScaleRGB=true
   )
   &FN_Light_Attributes(
    Color=$User.Color_02
    Create Color Attribute=true
    Create Exponent Attribute=true
    Create Radius Attribute=true
    Exponent=2
    Radius=15
    Scale Color=&FloatFromCurve(
     FloatCurve=DataInterface() {propertyValues {Curve {keys [
         <at 0 0.5 RCIM_Cubic>
         {arriveTangent=0.5, interpMode=RCIM_Cubic, leaveTangent=0.5, time=0.1, value=0.85}
         <at 0.2 0.6 RCIM_Cubic>
         {arriveTangent=-0.3333, interpMode=RCIM_Cubic, leaveTangent=-0.3333, time=0.4, value=0.85}
         {arriveTangent=0.25, interpMode=RCIM_Cubic, leaveTangent=0.25, time=0.5, value=0.5}
         {arriveTangent=-1.6667, interpMode=RCIM_Cubic, leaveTangent=-1.6667, time=0.6, value=0.9}
         <at 0.8 0>
        ]}}}
     Scale Curve=60
    )
   )
   &SkeletalMeshLocation(
    Bone / Socket Index=$Particles.SpawnGroup
    MaskSpawnGroup=false
    Mesh Sampling Type=::NewEnumerator5
    Sampled Position Offset=<20 0 0>
    Skeletal Mesh=$User.SkeletalMesh
    Socket Sampling Mode=::NewEnumerator5
    Use Sampled Mesh Attribute as Ribbon ID=false
   ) {enabled=false}
  ]
  rendererClasses [&NiagaraLightRendererProperties]
  renderers [{rendererClass=&NiagaraLightRendererProperties, rendererIndex=0}]
 }
 Lights {
  emitterUpdateScript [
   &EmitterState()
   &SpawnBurst_Instantaneous(Spawn Count=2, Spawn Time=0.001, Use Spawn Probability=false)
  ]
  particleSpawnScript [
   &InitializeParticle(
    Color=<rgba 12.6732 60.2039 125 1>
    Lifetime=1.4
    Module.Write Color=true
    Module.Write Lifetime=true
    Module.Write Mass=true
    Module.Write Material Random=true
    Module.Write Position=false
    Module.Write Scale=false
    Module.Write SpriteRotation=false
    Module.Write SpriteSize=true
    Position=<20 0 0>
    Sprite Size=<12 12>
   )
   &FN_SetSpawnGroup(Number of Spawn Groups=2)
  ]
  particleUpdateScript [
   &SkeletalMeshLocation(
    Apply Sampled Orientation to Mesh Particles=true
    Bone / Socket Index=&SpawnGroup()
    Inherit Velocity=false
    MaskSpawnGroup=false
    Mesh Sampling Type=::NewEnumerator5
    Sampled Position Offset=&MaskVectorBySpawnGroup(Masked Vector=<-5 -2 -5>, Passthrough Vector=<5 -2 -5>)
    Sampled Position Offset Coordinate Space:ENiagaraMeshTransforms=::NewEnumerator2
    Skeletal Mesh=$User.SkeletalMesh
    Socket Sampling Mode=::NewEnumerator5
   )
   /SetVariables(Particles.LightExponent=4) {isSetParameters=true}
   /SetVariables(Particles.LightRadius=1) {isSetParameters=true}
   &ParticleState(Kill Particles When Lifetime Has Elapsed=false, Let Infinitely Lived Particles Die When Emitter Deactivates=true)
   &FN_Light_Attributes(
    Color=<rgba 12.6732 60.2039 125 1>
    Create Color Attribute=true
    Create Exponent Attribute=true
    Create Radius Attribute=true
    Enable Particle Light Enable=true
    Light Enabled=true
    Radius=70
    Scale Color=&FloatFromCurve(
     FloatCurve=DataInterface() {propertyValues {Curve {keys [
         {interpMode=RCIM_Cubic, tangentMode=RCTM_Break, time=0, value=0}
         {interpMode=RCIM_Cubic, tangentMode=RCTM_Break, time=0.2, value=1}
         {arriveTangent=0.0038, interpMode=RCIM_Cubic, leaveTangent=0.0038, tangentMode=RCTM_User, time=0.8, value=1}
         <at 1 0 RCIM_Cubic>
        ]}}}
     Scale Curve=0.4
    )
   )
   &FN_Gen5_Light(
    Attenutation=30
    Color:LinearColor=<rgba 0.1014 0.4816 1 1>
    Exponent=5
    Intensity:NiagaraFloat=&Multiply_Float(
     A=&Multiply_Float(
      A=75
      B=&FloatFromCurve(FloatCurve=DataInterface() {propertyValues {Curve {keys [
           {interpMode=RCIM_Cubic, tangentMode=RCTM_Break, time=0, value=0}
           {interpMode=RCIM_Cubic, tangentMode=RCTM_Break, time=0.2, value=1}
           {arriveTangent=0.0038, interpMode=RCIM_Cubic, leaveTangent=0.0038, tangentMode=RCTM_User, time=0.8, value=1}
           <at 1 0 RCIM_Cubic>
          ]}}})
     )
     B=&Lerp_Float(A=0, B=1, Alpha=$User.LightSize)
    )
   )
  ]
  rendererClasses [
   &NiagaraSpriteRendererProperties
   &NiagaraLightRendererProperties
   &NiagaraComponentRendererProperties
  ]
  renderers [
   {rendererClass=&NiagaraSpriteRendererProperties}
   {rendererClass=&NiagaraLightRendererProperties, rendererIndex=1}
   {rendererClass=&NiagaraComponentRendererProperties, rendererIndex=2}
  ]
 }
}
summary {
 systemName=NS_Backpack_Skyfire_Boost_01
 emitters [
  {bEnabled=false, emitterName=Jet_Looping, rendererClasses [&NiagaraMeshRendererProperties]}
  {bEnabled=false, emitterName=Light, rendererClasses [&NiagaraLightRendererProperties]}
  {bEnabled=false, emitterName=ExhaustFlame_Small, rendererClasses [&NiagaraMeshRendererProperties]}
  {bEnabled=false, emitterName=ExhaustFlame_Small001, rendererClasses [&NiagaraMeshRendererProperties]}
  {emitterName=Flame, rendererClasses [&NiagaraMeshRendererProperties]}
  {emitterName=Jet, rendererClasses [&NiagaraSpriteRendererProperties]}
  {
   emitterName=Lights
   rendererClasses [
    &NiagaraSpriteRendererProperties
    &NiagaraLightRendererProperties
    &NiagaraComponentRendererProperties
   ]
  }
 ]
 userVariables [
  User.bisFrontEnd {type=NiagaraBool}
  User.bIsFrontEndPreview {type=NiagaraBool}
  User.bisFullyDeployed {type=NiagaraBool}
  User.Color_01 {type=LinearColor, description="", defaultValue {struct=LinearColor, value=<rgba 0.1216 0.0588 0.651 1>}}
  User.Color_02 {type=LinearColor, description="", defaultValue {struct=LinearColor, value=<rgba 0.15 0.2917 1 1>}}
  User.FlameScale {type=Vector3f, description="", defaultValue {struct=Vector3f, value=<1 1 1>}}
  User.ForwardDot {type=NiagaraFloat, defaultValue=0, description=""}
  User.LightSize {type=NiagaraFloat, defaultValue=1, description=""}
  User.RibbonIDOffset {type=NiagaraInt32}
  User.SkeletalMesh {
   type=NiagaraDataInterfaceSkeletalMesh
   defaultValue {
    struct=NiagaraExt_VariableValue_DataInterface
    value {dataInterface=&self:NiagaraDataInterfaceSkeletalMesh_2}
   }
  }
  User.SpawnRate {type=NiagaraFloat, defaultValue=1, description=""}
  User.SpriteFlame {type=NiagaraFloat, defaultValue=1, description=""}
  User.TrailColor {type=LinearColor, description="", defaultValue {struct=LinearColor, value=<rgba 0.9131 0.1845 0.8388 1>}}
 ]
}
