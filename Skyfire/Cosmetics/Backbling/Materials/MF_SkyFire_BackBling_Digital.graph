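# Skyfire back bling material function that draws a glowing digital timer and jump-count bar over incoming material attributes.
# Minutes and seconds index a digit flipbook per column, with a pulsing separator, faint ghost digit and optional flashing.
# Vertex red masks the jump bar, and a tinted mask texture supplies backpack glow.
# The combined result adds to emissive and drives a base color blend; the rune-count branch is left unconnected.
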MaterialFunction /Skyfire/Cosmetics/Backbling/Materials/MF_SkyFire_BackBling_Digital
graph {nodes [
  n1 = Multiply(A=Time(), B=ScalarParameter() {defaultValue=1, group=SkyFire, parameterName=SF_Tic_Speed})
  n2 = Multiply(A=Add(A=Sine(None=@n1))) {constB=0.5}
  n3 = TextureCoordinate() {coordinateIndex=1}
  n4 = VectorParameter() {group=SkyFire, parameterName=SF_Font_Offset}
  n5 = MaterialFunctionCall(
   UVs=Add(A=@n3, B=ComponentMask(None=@n4.RGB) {g=true, r=true})
   Texture Scale=ComponentMask(None=VectorParameter() {group=SkyFire, parameterName=SF_Font_Size, defaultValue=<rgba 1 1 0 1>}.RGB) {g=true, r=true}
  ) {materialFunction=&ScaleUVsByCenter}
  n6 = Add(A=@n5.UVs, B=ComponentMask(None=VectorParameter() {parameterName=SF_Offset, defaultValue=<rgba -0.025 0.01 0 1>}.RGB) {g=true, r=true})
  n7 = ComponentMask(None=@n6) {r=true}
  n8 = Divide(B=ScalarParameter() {defaultValue=10, group=SkyFire, parameterName=SF_No_Columns}) {constB=10}
  n9 = Multiply(A=ComponentMask(None=@n6) {g=true}, B=@n8)
  n10 = AppendVector(A=Add(A=Multiply(A=@n7) {constB=3}) {constB=0.5}, B=@n9)
  n11 = TextureObjectParameter() {
   group=SkyFire
   parameterName=SF_Font_Text
   samplerType=SAMPLERTYPE_Masks
   sortPriority=0
   texture=/Skyfire/Cosmetics/Backbling/Textures/T_NumberFlipBook_Digital
  }
  n12 = ComponentMask(None=@n5.UVs) {r=true}
  n13 = Add(A=Add(A=@n12) {constB=-1}, B=Divide() {constA=2.5, constB=3})
  n14 = Add(A=Add(A=@n12) {constB=-1}, B=Divide() {constA=1.5, constB=3})
  n15 = Subtract(A=Saturate(None=Divide(A=@n13) {constB=0.001}), B=Saturate(None=Divide(A=@n14) {constB=0.001}))
  n16 = Multiply(A=@n2, B=Multiply(A=@n15, B=TextureSample(UVs=@n10, Tex=@n11).G))
  n17 = ComponentMask(None=@n6) {r=true}
  n18 = ScalarParameter() {defaultValue=0.1, group=SkyFire, parameterName=SF_RT_Number_Offset}
  n19 = ComponentMask(None=ComponentMask(None=@n5.UVs) {r=true}) {r=true}
  n20 = Add(A=@n19) {constB=-1}
  n21 = Divide() {constB=3}
  n22 = Saturate(None=Divide(A=Add(A=@n20, B=@n21)) {constB=0.001})
  n23 = Frac(None=Add(A=Multiply(A=@n17) {constB=3}, B=Multiply(A=@n18, B=@n22)) {constB=0})
  n24 = ScalarParameter() {group=SkyFire, parameterName=SF_Number_Minutes}
  n25 = Subtract(A=@n24, B=Frac(None=@n24))
  n26 = ScalarParameter() {group=SkyFire, parameterName=SF_Number_Seconds}
  n27 = Divide(A=@n26) {constB=10}
  n28 = Subtract(A=@n27, B=Frac(None=@n27))
  n29 = Add(A=@n19) {constB=-1}
  n30 = Divide() {constA=2, constB=3}
  n31 = Saturate(None=Divide(A=Add(A=@n29, B=@n30)) {constB=0.001})
  n32 = Constant() {r=10}
  n33 = Divide(A=Add(A=Add(A=@n19) {constB=-1}, B=Divide() {constB=3})) {constB=0.001}
  n34 = Multiply(A=@n8, B=LinearInterpolate(A=LinearInterpolate(A=@n25, B=@n28, Alpha=@n31) {constB=0}, B=Fmod(A=@n26, B=@n32), Alpha=Saturate(None=@n33)))
  n35 = Multiply(A=@n3) {constB=0.5}
  n36 = DDX(Value=@n35)
  n37 = DDY(Value=@n35)
  n38 = TextureSample!(UVs=AppendVector(A=@n23, B=Add(A=@n9, B=@n34) {constB=0}), Tex=@n11, "DDX(UVs)"=@n36, "DDY(UVs)"=@n37, Apply View MipBias) {mipValueMode=TMVM_Derivative}
  n39 = Multiply(A=@n8) {constB=8}
  n40 = TextureSample(UVs=AppendVector(A=@n23, B=Add(A=@n9, B=@n39) {constB=2}), Tex=@n11, "DDX(UVs)"=@n36, "DDY(UVs)"=@n37) {mipValueMode=TMVM_Derivative}
  n41 = Multiply(
   A=Add(A=Add(A=@n16, B=@n38.R), B=Multiply(A=@n40.R) {constB=0.01})
   B=VectorParameter() {group=SkyFire, parameterName=SF_Font_Color, defaultValue=<rgba 0.4969 0.3916 0.807 1>}.RGB
  )
  n42 = Saturate(None=Add(A=@n2, B=ScalarParameter() {defaultValue=0.1, group=SkyFire, parameterName=SF_Min_Flash_Brightness, sortPriority=90}))
  n43 = LinearInterpolate(
   A=@n41
   B=Multiply(A=@n42, B=@n41)
   Alpha=ScalarParameter() {group=SkyFire, parameterName="SF_Flash_Timer 0=off 1=On", sliderMax=1, sortPriority=95}
  )
  n44 = VertexColor()
  n45 = OneMinus(None=Add(A=@n44.R, B=@n44.R))
  n46 = Add(A=OneMinus(None=ComponentMask(None=@n3) {r=true})) {constB=-1}
  n47 = Divide(A=ScalarParameter() {group=SkyFire, parameterName=SF_Jump_Count, sortPriority=85}) {constB=5}
  n48 = Saturate(None=Divide(A=Add(A=@n46, B=@n47)) {constB=0.001})
  n49 = VectorParameter() {group=SkyFire, parameterName=SF_Jump_Color, defaultValue=<rgba 0.0704 0.0437 0.1946 2.5>}
  n50 = Add(A=Multiply(A=Multiply(A=@n43, B=@n5.0-1 mask), B=@n45), B=Multiply(A=@n44.R, B=Multiply(A=@n48, B=@n49.RGB)))
  n51 = VectorParameter() {group=SkyFire, parameterName=SF_Rune_Color off, defaultValue=<rgba 0.9911 0.8308 1 1>}
  n52 = VectorParameter() {group=SkyFire, parameterName=SF_Rune_Color 0n, defaultValue=<rgba 0.2789 0.0467 0.8632 1>}
  n53 = Add(A=ComponentMask(None=@n3) {g=true}) {constB=-1}
  n54 = Divide(A=ScalarParameter() {group=SkyFire, parameterName=SF_Rune_Count}) {constB=4}
  n55 = Saturate(None=Divide(A=Add(A=@n53, B=@n54)) {constB=0})
  Add(A=@n50, B=Multiply(A=@n44.G, B=LinearInterpolate(A=@n51.RGB, B=@n52.RGB, Alpha=@n55)))
  n56 = DotProduct(A=CameraVectorWS(), B=VertexNormalWS())
  n57 = TextureSampleParameter2D(UVs=Panner(Coordinate=@n56) {speedX=0.025, speedY=0.05}) {group=SkyFire, parameterName=SF_Distortion_texture, sortPriority=0, texture=/Game/Effects/Fort_Effects/Textures/Tile/T_Caustics_Chromatic}
  n58 = Desaturation(None=@n57.RGB)
  n59 = Power(Base=@n56, Exp=ScalarParameter() {defaultValue=20, group=SkyFire, parameterName=SF_EmissiveInnerGlowAmount}) {constExponent=5}
  n60 = VectorParameter() {group=SkyFire, parameterName=SF_ColorTent, defaultValue=<rgba 0.1085 0.0674 0.3 1>}
  n61 = Multiply(A=Add(A=Multiply(A=@n58) {constB=0.25}, B=Multiply(A=@n59) {constB=0.25}), B=@n60.RGB)
  n62 = NamedRerouteUsage()
  n63 = Multiply(
   A=Multiply(A=@n60.RGB) {constB=1.5}
   B=Fresnel(ExponentIn=ScalarParameter() {defaultValue=8, group=SkyFire, parameterName=SF_EmissiveEdgeExponent}, Normal=Transform(None=@n62)) {baseReflectFraction=0}
  ) {constB=2}
  Add(A=Multiply(A=@n61) {constB=0.15}, B=@n63)
  n64 = Power(Base=@n56) {constExponent=100}
  Add(
   A=@n61
   B=Multiply(A=Multiply(A=@n64) {constB=0.1}, B=NamedRerouteDeclaration(None=@n60.RGB) {name=HueshiftColor, nodeColor=<rgba 0 1 0.9766 1>})
  )
  Add(A=@n15) {constB=0}
  n65 = GetMaterialAttributes(None=FunctionInput() {inputName=Material, inputType=FunctionInput_MaterialAttributes}) {attributeGetTypes [#1, #2]}
  n66 = TextureSampleParameter2D() {group=SkyFire, parameterName=SF_Mask_texture, sortPriority=0, texture=/Skyfire/Cosmetics/Backbling/Textures/T_SkyFire_Backpack_D}
  n67 = Multiply(A=@n60.RGB, B=Multiply(A=Power(Base=@n66.R) {constExponent=1.85}) {constB=20})
  n68 = LinearInterpolate(
   A=@n67
   B=Multiply(A=@n67, B=@n42)
   Alpha=ScalarParameter() {group=SkyFire, parameterName="SF_Flash_Backpack 0=off 1=On", sliderMax=1, sortPriority=100}
  )
  n69 = LinearInterpolate(A=Add(A=NamedRerouteUsage(), B=@n68), B=@n60.RGB, Alpha=NamedRerouteUsage())
  n70 = Add(A=Multiply(A=FeatureLevelSwitch(Default=@n69, ES2=@n68, ES3_1=@n68, SM4_REMOVED=@n69), B=@n45), B=@n50)
  n71 = NamedRerouteUsage()
  n72 = Constant3Vector()
  n73 = Multiply(A=Desaturation(None=@n70)) {constB=4}
  FunctionOutput(None=SetMaterialAttributes!(
    MaterialAttributes=@n65.MaterialAttributes
    EmissiveColor=Add(A=@n65.EmissiveColor, B=@n70)
    BaseColor=LinearInterpolate(A=@n71, B=@n72, Alpha=@n73)
   ) {attributeSetTypes [#1, #3]}) {outputName=Result_1}
  NamedRerouteDeclaration(None=@n65.Normal) {name=ObjectNormal, nodeColor=<rgba 0 0.7891 1 1>}
  NamedRerouteDeclaration(None=@n63) {name=EmissiveEdge, nodeColor=<rgba 0 1 0.25 1>}
  n74 = Power(Base=@n66.R) {constExponent=1.25}
  NamedRerouteDeclaration(None=Multiply(A=Multiply(A=@n74) {constB=10}, B=@n60.RGB)) {name=BaseColor, nodeColor=<rgba 0 0.7187 1 1>}
  n75 = Time()
  n76 = ScalarParameter() {defaultValue=1, group=SkyFire, parameterName=SF_Scanline_Speed}
  n77 = Fresnel() {baseReflectFraction=0, exponent=3}
  n78 = VertexNormalWS()
  n79 = CameraVectorWS()
  n80 = Multiply(
   A=DotProduct(A=ComponentMask(None=@n78) {b=true}, B=ComponentMask(None=@n79) {b=true})
   B=ScalarParameter() {defaultValue=0.5, group=SkyFire, parameterName=SF_Scanline_LineBend}
  )
  n81 = Constant()
  n82 = Add(
   A=QualitySwitch(Default=Multiply(A=@n77, B=@n80) {constB=0.2}, Low=@n81, Medium=@n81)
   B=ComponentMask(None=ScreenPosition().ViewportUV) {g=true}
  )
  n83 = ScalarParameter() {defaultValue=200, group=SkyFire, parameterName=SF_Scanline_Tile}
  n84 = Sine(None=Add(A=Multiply(A=@n75, B=@n76), B=Multiply(A=@n82, B=@n83)))
  n85 = LinearInterpolate(
   A=Power(Base=ConstantBiasScale(None=@n84))
   Alpha=MaterialFunctionCall(Distance=ScalarParameter() {defaultValue=400, group=SkyFire, parameterName=SF_Scanline_FadeDistance}) {materialFunction=&Near_Camera_Fade}.Result
  ) {constA=0.4}
  n86 = ScalarParameter() {defaultValue=0.005, group=SkyFire, parameterName=SF_Scanline_DarkenBrights}
  NamedRerouteDeclaration(None=Add(A=Multiply(A=@n85, B=@n86) {constB=0.25}, B=ScalarParameter() {group=SkyFire, parameterName=SF_Scanline_BrightenDarks})) {name=Scanlines, nodeColor=<rgba 1 0.3984 0 1>}
  Reroute(None=@n5.UVs)
 ]}
