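# Material function for the Skyfire back bling hologram, adding emissive color and setting opacity on incoming material attributes.
# Vertex green blends a Fresnel-faded, gradient tinted glow into star-textured runes, lit along UV1 by a rune count parameter.
# Flickering noise, disabled on Low quality, modulates scanlines in rune opacity.
# A timestamp parameter drives a brightness fade, opacity can flash on a sine, and leftover VendettaPuck color nodes feed no output.
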MaterialFunction /Skyfire/Cosmetics/Backbling/Materials/MF_SkyFire_BackBling_Holo
graph {nodes [
  n1 = ScalarParameter() {defaultValue=0.6, group=SkyFire, parameterName=Scan_MinScanlineOpacity}
  n2 = TextureCoordinate() {uTiling=0.05}
  n3 = AppendVector(
   A=ScalarParameter() {defaultValue=0.02, group=SkyFire, parameterName=Scan_NoiseMask-PanSpeed_X}
   B=ScalarParameter() {defaultValue=0.075, group=SkyFire, parameterName=Scan_NoiseMask-PanSpeed_Y}
  )
  n4 = Power(Base=TextureSample(UVs=Panner(Coordinate=@n2, Speed=@n3)) {samplerType=SAMPLERTYPE_LinearColor, texture=&T_NoisePack_LinearColor}.RGB) {constExponent=4}
  n5 = ScalarParameter() {defaultValue=1, group=SkyFire, parameterName=Scan_NoiseMaskMult}
  n6 = ScalarParameter() {group=SkyFire, parameterName=Scan_NoiseAdd}
  n7 = TextureCoordinate() {coordinateIndex=1}
  n8 = MaterialFunctionCall(Float2=Round(None=Divide(A=@n7) {constB=2000})) {materialFunction=&BreakOutFloat2Components}
  n9 = Add(A=Time(), B=Add(A=@n8.R, B=@n8.R))
  n10 = Multiply(A=ScalarParameter() {defaultValue=0.5, group=SkyFire, parameterName=Scan_FlickerSpeed}, B=@n9)
  n11 = AppendVector(A=AppendVector(A=Sine(None=@n10) {period=1.6}, B=Sine(None=@n10) {period=1.42}), B=Sine(None=@n10) {period=1.51})
  n12 = ScalarParameter() {defaultValue=0.5, group=SkyFire, parameterName=Scan_NoiseFrequency}
  n13 = Add(A=Sine(None=@n10) {period=2.1}, B=Multiply(B=@n12) {constA=0.25})
  n14 = Add(A=Sine(None=@n10) {period=0.72}, B=Multiply(B=@n12) {constA=0.75}) {constB=0.75}
  n15 = Ceil(None=Add(A=Multiply(B=@n12) {constA=0.9, constB=0.9}, B=Sine(None=@n10) {period=1.25}) {constB=0.9})
  n16 = Saturate(None=Multiply(A=Multiply(A=@n13, B=@n14), B=Multiply(A=@n14, B=@n15)))
  n17 = LinearInterpolate(Alpha=ConstantBiasScale(None=Sine(None=@n9) {period=0.075})) {constA=0.7}
  n18 = Multiply(A=Multiply(A=Ceil(None=@n11), B=Multiply(A=@n16, B=@n17)), B=ScalarParameter() {defaultValue=2, group=SkyFire, parameterName=Scan_NoiseMult})
  n19 = Saturate(None=Multiply(A=Multiply(A=@n4, B=@n5) {constB=20}, B=Add(A=@n6, B=@n18)))
  n20 = QualitySwitch(
   Default=MaterialFunctionCall(In=Desaturation(None=@n19), Contrast=ScalarParameter() {defaultValue=0.2, group=SkyFire, parameterName=Scan_NoiseContrast}) {materialFunction=&CheapContrast}.Result
   Low=Constant()
  )
  n21 = Multiply(A=@n20) {constB=1.5}
  n22 = Saturate(None=@n20)
  n23 = Multiply(
   A=ScalarParameter() {defaultValue=100, group=SkyFire, parameterName=Scan_ScanlineTile}
   B=LinearInterpolate(Alpha=Ceil(None=@n22)) {constA=1, constB=1.25}
  )
  n24 = TextureCoordinate() {coordinateIndex=1}
  n25 = ConstantBiasScale(None=Sine(None=Multiply(A=@n23, B=@n24) {constA=1}))
  n26 = LinearInterpolate(
   A=Clamp(None=Subtract(A=@n1, B=@n21)) {minDefault=-0.25}
   B=Add(A=@n21)
   Alpha=ComponentMask(None=Power(Base=@n25) {constExponent=1.5}) {g=true}
  ) {constA=0.5}
  ComponentMask(None=@n26) {g=true}
  n27 = GetMaterialAttributes(None=FunctionInput() {inputName=Material, inputType=FunctionInput_MaterialAttributes}) {attributeGetTypes [#1]}
  n28 = VectorParameter() {group=SkyFire, parameterName=SF_Holo_ColorTent, defaultValue=<rgba 0.1278 0 0.724 0.5>}
  n29 = VertexColor()
  n30 = VectorParameter() {group=SkyFire, parameterName=SF_Rune_Color 0n, defaultValue=<rgba 0.0937 0 0.5312 1>}
  n31 = TextureCoordinate() {coordinateIndex=1}
  n32 = ComponentMask(None=@n31) {g=true}
  n33 = Add(A=@n32) {constB=-1}
  n34 = Add(A=Divide(A=ScalarParameter() {group=SkyFire, parameterName=SF_Rune_Count}) {constB=5}) {constB=0.2}
  n35 = Saturate(None=Divide(A=Add(A=@n33, B=@n34)) {constB=0})
  n36 = TextureSample(UVs=@n31) {texture=/Skyfire/Cosmetics/Backbling/Effects/Hologram/T_FX_Skyfire_Backpack_Stars}
  n37 = LinearInterpolate(
   A=LinearInterpolate(A=VectorParameter() {group=SkyFire, parameterName=SF_Rune_Color off, defaultValue=<rgba 0.1927 0.1927 0.1927 1>}.RGB, B=@n30.RGB, Alpha=@n35)
   B=Multiply(A=@n30.RGB, B=ScalarParameter() {defaultValue=2, group=SkyFire, parameterName=SF_Rune_outline multplyer}) {constB=4}
   Alpha=@n36.R
  )
  n38 = Divide(
   A=Subtract(A=Time(), B=ScalarParameter() {defaultValue=-9999999, group=SkyFire, parameterName=SF_Holo_Time_BP_Brightness})
   B=ScalarParameter() {defaultValue=0.25, group=SkyFire, parameterName=SF_Holo_Time_BP_Duration_Brightness}
  )
  n39 = Multiply(
   A=LinearInterpolate(A=@n28.RGB, B=Multiply(A=@n29.G, B=@n37), Alpha=@n29.G)
   B=LinearInterpolate(
    A=ScalarParameter() {defaultValue=4, group=SkyFire, parameterName=SF_Holo_Time_Brightness__Start}
    B=ScalarParameter() {defaultValue=1, group=SkyFire, parameterName=SF_Holo_Time_Brightness__End}
    Alpha=Saturate(None=@n38)
   )
  )
  n40 = ScalarParameter() {defaultValue=1, group=SkyFire, parameterName=SF_Holo_Opacity}
  n41 = Fresnel(
   ExponentIn=ScalarParameter() {defaultValue=1, group=SkyFire, parameterName=SF_Fresnel_ExponentIn}
   BaseReflectFractionIn=ScalarParameter() {group=SkyFire, parameterName=SF_Fresnel_BaseReflectFractionIn}
  )
  n42 = Saturate(None=Power(Base=@n32, Exp=ScalarParameter() {defaultValue=2, group=SkyFire, parameterName=SF_Holo_Contrast}))
  n43 = Add(
   A=@n36.R
   B=LinearInterpolate(
    A=ScalarParameter() {defaultValue=0.25, group=SkyFire, parameterName=SF_Rune_opacty_off}
    B=ScalarParameter() {defaultValue=1, group=SkyFire, parameterName=SF_Rune_opacty_on}
    Alpha=@n35
   )
  )
  n44 = LinearInterpolate(A=Multiply(A=OneMinus(None=@n41), B=Multiply(A=@n28.A, B=@n42)), B=Multiply(A=Saturate(None=@n43), B=@n26), Alpha=@n29.G)
  n45 = Sine(None=Multiply(A=Time(), B=ScalarParameter() {defaultValue=1, group=SkyFire, parameterName=SF_Tic_Speed}))
  n46 = Add(A=Multiply(A=Add(A=@n45)) {constB=0.5}, B=ScalarParameter() {defaultValue=0.1, group=SkyFire, parameterName=SF_Min_Flash_Brightness})
  n47 = LinearInterpolate(
   A=@n44
   B=Multiply(A=@n44, B=Saturate(None=@n46))
   Alpha=ScalarParameter() {group=SkyFire, parameterName="SF_Flash_Backpack 0=off 1=On", sliderMax=1}
  )
  FunctionOutput(None=SetMaterialAttributes!(MaterialAttributes=@n27.MaterialAttributes, EmissiveColor=Add(A=@n27.EmissiveColor, B=@n39), Opacity=Multiply(A=@n40, B=@n47)) {attributeSetTypes [#1, #2]}) {outputName=Result_1}
  n48 = Add(A=ScalarParameter() {defaultValue=0.7, parameterName=ScanlineColorMinInfluence}, B=@n25)
  n49 = ScalarParameter() {defaultValue=-0.4, group=Noise, parameterName=NoiseBrightness}
  n50 = Saturate(None=TextureCoordinate() {coordinateIndex=1})
  n51 = LinearInterpolate(
   A=Multiply(A=Multiply(A=@n49, B=@n22) {constB=0.1}, B=Saturate(None=@n48))
   B=LinearInterpolate(A=@n50, B=OneMinus(None=@n50), Alpha=ScalarParameter() {parameterName=FlipGradient})
   Alpha=ScalarParameter() {defaultValue=0.4, parameterName=VerticalGradientInflience}
  )
  n52 = TextureSampleParameter2D() {group=Texture, parameterName=MaskTexture, texture=/Engine/EngineResources/WhiteSquareTexture}
  n53 = Multiply(
   A=LinearInterpolate(
    A=ScalarParameter() {parameterName=MaskMin}
    Alpha=ChannelMaskParameter(None=@n52.RGB) {maskChannel=Green, parameterName=MaskChannel, defaultValue=<rgba 0 1 0 0>}.R
   )
   B=Multiply(A=ScalarParameter() {defaultValue=1, parameterName=TurnOn}, B=@n51)
  )
  Multiply(A=Saturate(None=Multiply(A=@n53, B=@n26)), B=ScalarParameter() {defaultValue=1.25, parameterName=OpacityMult})
  n54 = TextureSampleParameter2D() {group=Diffuse Texture Overlay, parameterName=Diffuse Texture Overlay, texture=&WhiteSquareTexture}
  n55 = MaterialFunctionCall(In=@n54.RGB, Contrast=ScalarParameter() {group=Diffuse Texture Overlay, parameterName=Diffuse Contribution Cheap Contrast}) {materialFunction=&CheapContrast}
  n56 = ScalarParameter() {group=Diffuse Texture Overlay, parameterName=Diffuse Overlay Use Cheap Contrast}
  n57 = LinearInterpolate(
   A=VectorParameter() {group=VendettaPuck, parameterName=SlateColor-Dark, defaultValue=<rgba 0 0 0 1>}.RGB
   B=VectorParameter() {group=VendettaPuck, parameterName=SlateColor-Bright, defaultValue=<rgba 0.0298 0.0866 0.2 1>}.RGB
   Alpha=Power(Base=ComponentMask(None=@n51) {g=true}, Exp=ScalarParameter() {defaultValue=1, group=VendettaPuck, parameterName=SlateColor_LerpCurvePower})
  )
  n58 = Multiply(
   A=Desaturation(
    None=LinearInterpolate(A=@n54.RGB, B=@n55.Result, Alpha=@n56)
    Fraction=ScalarParameter() {defaultValue=1, group=Diffuse Texture Overlay, parameterName=Diffuse Contribution Desaturation}
   )
   B=@n57
  )
  n59 = ScalarParameter() {group=Diffuse Texture Overlay, parameterName=Lerp Diffuse Overlay}
  Multiply(A=Multiply(A=LinearInterpolate(A=@n57, B=@n58, Alpha=@n59), B=ScalarParameter() {defaultValue=2, parameterName=EmissiveMult}) {constB=2})
  TextureCoordinate()
 ]}
