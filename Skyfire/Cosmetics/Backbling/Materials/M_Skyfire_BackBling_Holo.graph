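# Material for the holographic Skyfire back bling cosmetic, using material attributes and Default Lit shading, usable on skeletal meshes and Niagara mesh particles.
# The shared character PreFX function feeds MF_SkyFire_BackBling_Holo, which supplies the holo-specific stage.
# Its result passes into the shared character PostFX function alongside PreFX outputs.
# A separate CharacterSkinFX node also reads PreFX's pre-skinned position, with no consumer shown.
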Material /Skyfire/Cosmetics/Backbling/Materials/M_Skyfire_BackBling_Holo
graph {
 properties {
  bUseMaterialAttributes=true
  bUsedWithMeshDeformer=true
  bUsedWithNiagaraMeshParticles=true
  bUsedWithSkeletalMesh=true
  pixelDepthOffsetMode=PDOM_Legacy
  refractionMethod=RM_IndexOfRefraction
  usedShadingModels=Default Lit
 }
 nodes [
  n1 = MaterialFunctionCall() {materialFunction=&BaseFN_CharacterFunction_PreFX}
  MaterialFunctionCall(
   AO=@n1.AO
   Distance Blend=@n1.AO
   Fresnel=@n1.AO
   Gmap=@n1.AO
   Pre Skinned Local Position=@n1.AO
   Material=MaterialFunctionCall!(Material=@n1.Result) {materialFunction=&/MF_SkyFire_BackBling_Holo}.Result_1
  ) {materialFunction=&BaseFN_CharacterFunction_PostFX}
  MaterialFunctionCall(
   PreskinnedPosition=@n1.Pre skinned Local Position
   Gmap=@n1.Pre skinned Local Position
   MaterialAttributes=@n1.Pre skinned Local Position
   AO=@n1.Pre skinned Local Position
  ) {materialFunction=&CharacterSkinFX}
 ]
}
